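# Physics asset for the M_MED_Dusty male player body that drives its dangling gear: pouches, radio, holster, straps, clip, headpiece, wires and cable loop.
# Each dyn bone has a capsule body, mostly constrained to a tiny matching _MSTR anchor body.
# Rear pouch zippers and the second front strap chain off their parent pieces.
# Radio and pouches lock both swings, leaving twist, while the back wire rotates freely on all axes.
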PhysicsAsset /Game/Characters/Player/Male/Medium/Bodies/M_MED_Dusty/Meshes/M_MED_Dusty_Dyn
bodies {
 dyn_belt_pouch_l_01 [{shapeType=Capsule, length=4.2812, radius=5.4375, center=<1.6793 0.0001 3>, rotation=<rot 90 0 0>}]
 dyn_belt_pouch_l_01_MSTR [{shapeType=Capsule, length=0.505, radius=0.505}]
 dyn_belt_pouch_l_02 [{shapeType=Capsule, length=4.2812, radius=5.4375, center=<1.6793 0.0001 3>, rotation=<rot 90 0 0>}]
 dyn_belt_pouch_l_02_MSTR [{shapeType=Capsule, length=0.505, radius=0.505}]
 dyn_cable_loop_bk [{shapeType=Capsule, length=3.9688, radius=3.6562, center=<-2.1314 0 0.0003>, rotation=<rot 90 0 0>}]
 dyn_cable_loop_bk_MSTR [{shapeType=Capsule, length=0.505, radius=0.505}]
 dyn_clip_fr_01 [{shapeType=Capsule, length=4.2812, radius=4.375, center=<1 0 0>, rotation=<rot 90 0 0>}]
 dyn_clip_fr_01_MSTR [{shapeType=Capsule, length=0.505, radius=0.505}]
 dyn_headpiece_01 [{shapeType=Capsule, length=4.2812, radius=4.375, center=<1 0 0>, rotation=<rot 90 0 0>}]
 dyn_headpiece_01_MSTR [{shapeType=Capsule, length=0.505, radius=0.505}]
 dyn_holster [{shapeType=Capsule, length=4.2812, radius=4.375, center=<1 0 0>, rotation=<rot 90 0 0>}]
 dyn_holster_MSTR [{shapeType=Capsule, length=0.505, radius=0.505}]
 dyn_pouch_bk_r [{shapeType=Capsule, length=6.1875, radius=6.1875, center=<1.3403 0 2.2>, rotation=<rot 74.9999 0 0.0001>}]
 dyn_pouch_bk_r_MSTR [{shapeType=Capsule, length=0.505, radius=0.505}]
 dyn_pouch_zippers_bk_r [{shapeType=Capsule, length=2.0625, radius=4.1562, center=<1 0 0>, rotation=<rot 90 0 0>}]
 dyn_radio [{shapeType=Capsule, length=9.8438, radius=4.9062, center=<4.387 0 2.5>, rotation=<rot 90 0 0>}]
 dyn_radio_MSTR [{shapeType=Capsule, length=0.505, radius=0.505}]
 dyn_strap_fr_01 [{shapeType=Capsule, length=3.3125, radius=3.4062, center=<1 0 0>, rotation=<rot 90 0 0>}]
 dyn_strap_fr_01_MSTR [{shapeType=Capsule, length=0.505, radius=0.505}]
 dyn_strap_fr_02 [{shapeType=Capsule, length=3.3125, radius=3.4062, center=<1 0 0>, rotation=<rot 90 0 0>}]
 dyn_wire_bk [{shapeType=Capsule, length=2, radius=3.25, rotation=<rot 90 0 0>}]
 dyn_wire_bk_MSTR [{shapeType=Capsule, length=0.505, radius=0.505}]
 dyn_wire_hanging_bk [{shapeType=Capsule, length=3.9688, radius=3.6562, center=<1 0 0>, rotation=<rot 90 0 0>}]
 dyn_wire_hanging_bk_MSTR [{shapeType=Capsule, length=0.505, radius=0.505}]
}
constraints [
 {
  bone1Name=dyn_wire_bk
  bone2Name=dyn_wire_bk_MSTR
  swing1LimitDegrees=45
  swing1Motion=Free
  swing2LimitDegrees=45
  swing2Motion=Free
  twistLimitDegrees=5
  twistMotion=Free
 }
 {bone1Name=dyn_wire_hanging_bk, bone2Name=dyn_wire_hanging_bk_MSTR, swing1LimitDegrees=45, swing2LimitDegrees=20, twistLimitDegrees=5}
 {bone1Name=dyn_cable_loop_bk, bone2Name=dyn_cable_loop_bk_MSTR, swing1LimitDegrees=45, swing2LimitDegrees=10, twistLimitDegrees=5}
 {
  bone1Name=dyn_radio
  bone2Name=dyn_radio_MSTR
  swing1LimitDegrees=30
  swing1Motion=Locked
  swing2LimitDegrees=5
  swing2Motion=Locked
  twistLimitDegrees=15
 }
 {bone1Name=dyn_headpiece_01, bone2Name=dyn_headpiece_01_MSTR, swing1LimitDegrees=45, swing2LimitDegrees=10, twistLimitDegrees=5}
 {bone1Name=dyn_clip_fr_01, bone2Name=dyn_clip_fr_01_MSTR, swing1LimitDegrees=45, swing2LimitDegrees=10, twistLimitDegrees=5}
 {
  bone1Name=dyn_pouch_bk_r
  bone2Name=dyn_pouch_bk_r_MSTR
  swing1LimitDegrees=45
  swing1Motion=Locked
  swing2LimitDegrees=45
  swing2Motion=Locked
  twistLimitDegrees=10
 }
 {bone1Name=dyn_pouch_zippers_bk_r, bone2Name=dyn_pouch_bk_r, swing1LimitDegrees=25, swing2LimitDegrees=15, twistLimitDegrees=5}
 {
  bone1Name=dyn_belt_pouch_l_01
  bone2Name=dyn_belt_pouch_l_01_MSTR
  swing1LimitDegrees=5
  swing1Motion=Locked
  swing2LimitDegrees=15
  swing2Motion=Locked
  twistLimitDegrees=10
 }
 {
  bone1Name=dyn_belt_pouch_l_02
  bone2Name=dyn_belt_pouch_l_02_MSTR
  swing1LimitDegrees=5
  swing1Motion=Locked
  swing2LimitDegrees=15
  swing2Motion=Locked
  twistLimitDegrees=10
 }
 {bone1Name=dyn_strap_fr_01, bone2Name=dyn_strap_fr_01_MSTR, swing1LimitDegrees=45, swing2LimitDegrees=25, twistLimitDegrees=5}
 {bone1Name=dyn_strap_fr_02, bone2Name=dyn_strap_fr_01, swing1LimitDegrees=45, swing2LimitDegrees=25, twistLimitDegrees=5}
 {bone1Name=dyn_holster, bone2Name=dyn_holster_MSTR, swing1LimitDegrees=15, swing2LimitDegrees=15, twistLimitDegrees=5}
]
