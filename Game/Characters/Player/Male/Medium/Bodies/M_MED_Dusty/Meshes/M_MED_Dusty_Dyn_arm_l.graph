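# Physics asset for the left-arm strap on the Dusty medium male player body, providing simulation shapes for that dynamic accessory.
# It has two capsule bodies: a larger, offset and rotated one on the strap bone and a tiny one on its master bone.
# A single constraint links the strap to the master bone with tight swing and twist limits, allowing only slight motion.
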PhysicsAsset /Game/Characters/Player/Male/Medium/Bodies/M_MED_Dusty/Meshes/M_MED_Dusty_Dyn_arm_l
bodies {
 dyn_arm_strap_l [{shapeType=Capsule, length=3.625, radius=3.6562, center=<0.5 0 0>, rotation=<rot 90 0 0>}]
 dyn_arm_strap_l_MSTR [{shapeType=Capsule, length=0.505, radius=0.505}]
}
constraints [{bone1Name=dyn_arm_strap_l, bone2Name=dyn_arm_strap_l_MSTR, swing1LimitDegrees=10, swing2LimitDegrees=20, twistLimitDegrees=5}]
