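# Physics asset for the right-arm dynamic strap on the M_MED_Dusty medium male player body.
# It holds two capsule bodies, one on the strap bone and a much smaller one on its MSTR bone.
# A single constraint joins them with small swing and twist limits, keeping the strap's movement tightly bounded relative to that bone.
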PhysicsAsset /Game/Characters/Player/Male/Medium/Bodies/M_MED_Dusty/Meshes/M_MED_Dusty_Dyn_arm_r
bodies {
 dyn_arm_strap_r [{shapeType=Capsule, length=4.0312, radius=4.0312, center=<0.5 0 0>, rotation=<rot 90 0 0>}]
 dyn_arm_strap_r_MSTR [{shapeType=Capsule, length=0.505, radius=0.505}]
}
constraints [{bone1Name=dyn_arm_strap_r, bone2Name=dyn_arm_strap_r_MSTR, swing1LimitDegrees=10, swing2LimitDegrees=20, twistLimitDegrees=5}]
