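# Physics asset for the dynamic shirt on the medium male player body Dusty, covering five shirt bones at the back and sides.
# Each shirt capsule joins a tiny MSTR capsule with limited swing and tight twist, and each MSTR joins the pelvis with free rotation.
# Larger pelvis and thigh capsules give the lower-body shape, with both thighs also joined to the pelvis by free-rotation constraints.
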PhysicsAsset /Game/Characters/Player/Male/Medium/Bodies/M_MED_Dusty/Meshes/M_MED_Dusty_Dyn_shirt
bodies {
 dyn_shirt_bk [{shapeType=Capsule, length=7.5, radius=5.625, center=<3.6 0 5.6>, rotation=<rot 90 0 0>}]
 dyn_shirt_bk_MSTR [{shapeType=Capsule, length=0.505, radius=0.505}]
 dyn_shirt_bk_l [{shapeType=Capsule, length=8.125, radius=5, center=<3.6 0 5.6>, rotation=<rot 90 0 0>}]
 dyn_shirt_bk_l_MSTR [{shapeType=Capsule, length=0.505, radius=0.505}]
 dyn_shirt_bk_r [{shapeType=Capsule, length=8.125, radius=5, center=<3.6 0 5.6>, rotation=<rot 90 0 0>}]
 dyn_shirt_bk_r_MSTR [{shapeType=Capsule, length=0.505, radius=0.505}]
 dyn_shirt_l [{shapeType=Capsule, length=7.3125, radius=5.8125, center=<3.6 0 5.6>, rotation=<rot 90 0 0>}]
 dyn_shirt_l_MSTR [{shapeType=Capsule, length=0.505, radius=0.505}]
 dyn_shirt_r [{shapeType=Capsule, length=7.3125, radius=5.8125, center=<3.6 0 5.6>, rotation=<rot 90 0 0>}]
 dyn_shirt_r_MSTR [{shapeType=Capsule, length=0.505, radius=0.505}]
 pelvis [{shapeType=Capsule, length=8.5771, radius=11.769, center=<-2.7428 -0.7511 0>}]
 thigh_l [{shapeType=Capsule, length=28.8583, radius=9.4565, center=<-14.0497 -1.5416 0>, rotation=<rot 90 0 0>}]
 thigh_r [{shapeType=Capsule, length=28.8583, radius=9.4565, center=<14.0497 1.5416 0>, rotation=<rot 90 0 -179.9999>}]
}
constraints [
 {bone1Name=dyn_shirt_bk, bone2Name=dyn_shirt_bk_MSTR, swing1LimitDegrees=45, swing2LimitDegrees=45, twistLimitDegrees=5}
 {bone1Name=dyn_shirt_bk_l, bone2Name=dyn_shirt_bk_l_MSTR, swing1LimitDegrees=45, swing2LimitDegrees=45, twistLimitDegrees=5}
 {bone1Name=dyn_shirt_l, bone2Name=dyn_shirt_l_MSTR, swing1LimitDegrees=45, swing2LimitDegrees=45, twistLimitDegrees=5}
 {bone1Name=dyn_shirt_bk_r, bone2Name=dyn_shirt_bk_r_MSTR, swing1LimitDegrees=45, swing2LimitDegrees=45, twistLimitDegrees=5}
 {bone1Name=dyn_shirt_r, bone2Name=dyn_shirt_r_MSTR, swing1LimitDegrees=45, swing2LimitDegrees=45, twistLimitDegrees=5}
 {
  bone1Name=dyn_shirt_bk_MSTR
  bone2Name=pelvis
  swing1LimitDegrees=45
  swing1Motion=Free
  swing2LimitDegrees=45
  swing2Motion=Free
  twistLimitDegrees=45
  twistMotion=Free
 }
 {
  bone1Name=dyn_shirt_bk_l_MSTR
  bone2Name=pelvis
  swing1LimitDegrees=45
  swing1Motion=Free
  swing2LimitDegrees=45
  swing2Motion=Free
  twistLimitDegrees=45
  twistMotion=Free
 }
 {
  bone1Name=dyn_shirt_l_MSTR
  bone2Name=pelvis
  swing1LimitDegrees=45
  swing1Motion=Free
  swing2LimitDegrees=45
  swing2Motion=Free
  twistLimitDegrees=45
  twistMotion=Free
 }
 {
  bone1Name=dyn_shirt_bk_r_MSTR
  bone2Name=pelvis
  swing1LimitDegrees=45
  swing1Motion=Free
  swing2LimitDegrees=45
  swing2Motion=Free
  twistLimitDegrees=45
  twistMotion=Free
 }
 {
  bone1Name=dyn_shirt_r_MSTR
  bone2Name=pelvis
  swing1LimitDegrees=45
  swing1Motion=Free
  swing2LimitDegrees=45
  swing2Motion=Free
  twistLimitDegrees=45
  twistMotion=Free
 }
 {
  bone1Name=thigh_r
  bone2Name=pelvis
  swing1LimitDegrees=45
  swing1Motion=Free
  swing2LimitDegrees=45
  swing2Motion=Free
  twistLimitDegrees=45
  twistMotion=Free
 }
 {
  bone1Name=thigh_l
  bone2Name=pelvis
  swing1LimitDegrees=45
  swing1Motion=Free
  swing2LimitDegrees=45
  swing2Motion=Free
  twistLimitDegrees=45
  twistMotion=Free
 }
]
