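# Animation Blueprint for the Dusty player body, adding deformation and secondary motion to the parent mesh's copied pose.
# Twist and sculpter Control Rigs feed weighted constraints placing gear bones, then four RigidBody passes simulate dynamics, arms and shirt.
# Each update computes gravity override and simulation-space rotation through deprecated helper functions, driving the physics.
# A backbling check collapses the radio and back pouch bones to zero scale and sets a material hide parameter.
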AnimBlueprint /Game/Characters/Player/Male/Medium/Bodies/M_MED_Dusty/Meshes/M_MED_Dusty_AnimBP
target_skeleton=&/M_MED_Dusty_Skeleton
graphs [
 AnimGraph {nodes [
   n1 = Control Rig(Source=Copy Pose From Mesh() {bUseAttachedParent=true}) {
    alphaBoolBlend {blendOption=Linear}
    controlRigAssetReference {blueprintRigClass=&/M_MED_Dusty_Deform_CtrlRig_twists.M_MED_Dusty_Deform_CtrlRig_twists_C}
   }
   n2 = Control Rig!(
    Source=@n1
    Alpha
    KneeRot_fr_l_114deg:Vector
    KneeTrans_fr_l_114deg:Vector="0.258,-3.591,0.009"
    KneeScale_fr_l_114deg:Vector="1,1,1"
    KneeScale_fr_l_130deg:Vector="1,1,1"
    KneeRot_fr_l_130deg:Vector="0,0,-51.2"
    KneeTrans_fr_l_130deg:Vector="0.258,-3.591,0.009"
   ) -> (Pose) {
    alphaBoolBlend {blendOption=Linear}
    controlRigAssetReference {blueprintRigClass=&/M_MED_Dusty_Deform_CtrlRig_sculpters.M_MED_Dusty_Deform_CtrlRig_sculpters_C}
   }
   n3 = "Constraint - dyn_holster_MSTR (PPPPP)"!(
    ComponentPose=Local To Component(LocalPose=@n2)
    Alpha=1
    ConstraintWeights_0=0.0531
    ConstraintWeights_1:float=0.0729
    ConstraintWeights_2:float=0.3097
    ConstraintWeights_3:float=0.2977
    ConstraintWeights_4:float=0.2666
   ) -> (Pose) {
    alphaBoolBlend {blendOption=Linear}
    boneToModify {boneName=dyn_holster_MSTR}
    constraintSetup [
     {targetBone {boneName=spine_01}}
     {targetBone {boneName=deform_thigh_bk_r}}
     {targetBone {boneName=deform_thigh_fr_r}}
     {targetBone {boneName=thigh_twist_01_r}}
     {targetBone {boneName=pelvis}}
    ]
    constraintWeights [1,1,1,1,1]
   }
   n4 = "Constraint - dyn_arm_strap_r_MSTR (PPPPP)"!(
    ComponentPose=@n3
    Alpha=1
    ConstraintWeights_0=0.3742
    ConstraintWeights_1:float=0.0026
    ConstraintWeights_2:float=0.1676
    ConstraintWeights_3:float=0.4346
    ConstraintWeights_4:float=0.021
   ) -> (Pose) {
    alphaBoolBlend {blendOption=Linear}
    boneToModify {boneName=dyn_arm_strap_r_MSTR}
    constraintSetup [
     {targetBone {boneName=lowerarm_twist_02_r}}
     {targetBone {boneName=upperarm_twist_02_r}}
     {targetBone {boneName=upperarm_r}}
     {targetBone {boneName=lowerarm_r}}
     {targetBone {boneName=lowerarm_twist_01_r}}
    ]
    constraintWeights [1,1,1,1,1]
   }
   n5 = "Constraint - dyn_arm_strap_l_MSTR (PPPPP)"!(
    ComponentPose=@n4
    Alpha=1
    ConstraintWeights_0=0.4346
    ConstraintWeights_1:float=0.021
    ConstraintWeights_2:float=0.1676
    ConstraintWeights_3:float=0.3742
    ConstraintWeights_4:float=0.0026
   ) -> (Pose) {
    alphaBoolBlend {blendOption=Linear}
    boneToModify {boneName=dyn_arm_strap_l_MSTR}
    constraintSetup [
     {targetBone {boneName=lowerarm_l}}
     {targetBone {boneName=lowerarm_twist_01_l}}
     {targetBone {boneName=upperarm_l}}
     {targetBone {boneName=lowerarm_twist_02_l}}
     {targetBone {boneName=upperarm_twist_02_l}}
    ]
    constraintWeights [1,1,1,1,1]
   }
   n6 = "Constraint - dyn_strap_fr_01_MSTR (PPPPP)"!(
    ComponentPose=@n5
    Alpha=1
    ConstraintWeights_0=0.3631
    ConstraintWeights_1:float=0.1679
    ConstraintWeights_2:float=0.1085
    ConstraintWeights_3:float=0.0264
    ConstraintWeights_4:float=0.3341
   ) -> (Pose) {
    alphaBoolBlend {blendOption=Linear}
    boneToModify {boneName=dyn_strap_fr_01_MSTR}
    constraintSetup [
     {targetBone {boneName=pelvis}}
     {targetBone {boneName=spine_02}}
     {targetBone {boneName=spine_03}}
     {targetBone {boneName=spine_04}}
     {targetBone {boneName=spine_01}}
    ]
    constraintWeights [1,1,1,1,1]
   }
   n7 = "Constraint - dyn_pouch_bk_r_MSTR (PPPPPP)"!(
    ComponentPose=@n6
    Alpha=1
    ConstraintWeights_0=0.1194
    ConstraintWeights_1:float=0.2357
    ConstraintWeights_2:float=0.2854
    ConstraintWeights_3:float=0.2198
    ConstraintWeights_4:float=0.0855
    ConstraintWeights_5:float=0.0543
   ) -> (Pose) {
    alphaBoolBlend {blendOption=Linear}
    boneToModify {boneName=dyn_pouch_bk_r_MSTR}
    constraintSetup [
     {targetBone {boneName=spine_01}}
     {targetBone {boneName=spine_02}}
     {targetBone {boneName=spine_03}}
     {targetBone {boneName=spine_04}}
     {targetBone {boneName=spine_05}}
     {targetBone {boneName=pelvis}}
    ]
    constraintWeights [1,1,1,1,1,1]
   }
   n8 = "Constraint - dyn_belt_pouch_l_02_MSTR (PPPPPP)"!(
    ComponentPose=@n7
    Alpha=1
    ConstraintWeights_0=0.1747
    ConstraintWeights_1:float=0.2495
    ConstraintWeights_2:float=0.2451
    ConstraintWeights_3:float=0.2021
    ConstraintWeights_4:float=0.0103
    ConstraintWeights_5:float=0.1183
   ) -> (Pose) {
    alphaBoolBlend {blendOption=Linear}
    boneToModify {boneName=dyn_belt_pouch_l_02_MSTR}
    constraintSetup [
     {targetBone {boneName=spine_01}}
     {targetBone {boneName=spine_02}}
     {targetBone {boneName=spine_03}}
     {targetBone {boneName=spine_04}}
     {targetBone {boneName=spine_05}}
     {targetBone {boneName=pelvis}}
    ]
    constraintWeights [1,1,1,1,1,1]
   }
   n9 = "Constraint - dyn_belt_pouch_l_01_MSTR (PPPPPP)"!(
    ComponentPose=@n8
    Alpha=1
    ConstraintWeights_0=0.262
    ConstraintWeights_1:float=0.2175
    ConstraintWeights_2:float=0.1737
    ConstraintWeights_3:float=0.1243
    ConstraintWeights_4:float=0.0166
    ConstraintWeights_5:float=0.2059
   ) -> (Pose) {
    alphaBoolBlend {blendOption=Linear}
    boneToModify {boneName=dyn_belt_pouch_l_01_MSTR}
    constraintSetup [
     {targetBone {boneName=spine_01}}
     {targetBone {boneName=spine_02}}
     {targetBone {boneName=spine_03}}
     {targetBone {boneName=spine_04}}
     {targetBone {boneName=spine_05}}
     {targetBone {boneName=pelvis}}
    ]
    constraintWeights [1,1,1,1,1,1]
   }
   n10 = "Constraint - dyn_radio_MSTR (PPPPP)"!(
    ComponentPose=@n9
    Alpha=1
    ConstraintWeights_0=0.0144
    ConstraintWeights_1:float=0.1029
    ConstraintWeights_2:float=0.1647
    ConstraintWeights_3:float=0.3827
    ConstraintWeights_4:float=0.3353
   ) -> (Pose) {
    alphaBoolBlend {blendOption=Linear}
    boneToModify {boneName=dyn_radio_MSTR}
    constraintSetup [
     {targetBone {boneName=spine_01}}
     {targetBone {boneName=spine_02}}
     {targetBone {boneName=spine_03}}
     {targetBone {boneName=spine_04}}
     {targetBone {boneName=spine_05}}
    ]
    constraintWeights [1,1,1,1,1]
   }
   n11 = "Constraint - dyn_cable_loop_bk (PPPP)"!(
    ComponentPose=@n10
    Alpha=1
    ConstraintWeights_0=0.0169
    ConstraintWeights_1:float=0.1993
    ConstraintWeights_2:float=0.6731
    ConstraintWeights_3:float=0.1107
   ) -> (Pose) {
    alphaBoolBlend {blendOption=Linear}
    boneToModify {boneName=dyn_cable_loop_bk}
    constraintSetup [{targetBone {boneName=spine_03}}, {targetBone {boneName=spine_04}}, {targetBone {boneName=spine_05}}, {targetBone {boneName=clavicle_l}}]
    constraintWeights [1,1,1,1]
   }
   n12 = "Constraint - dyn_wire_hanging_bk_MSTR (PPP)"!(ComponentPose=@n11, Alpha=1, ConstraintWeights_0=0.4107, ConstraintWeights_1:float=0.5202, ConstraintWeights_2:float=0.0691) -> (Pose) {
    alphaBoolBlend {blendOption=Linear}
    boneToModify {boneName=dyn_wire_hanging_bk_MSTR}
    constraintSetup [{targetBone {boneName=spine_04}}, {targetBone {boneName=spine_05}}, {targetBone {boneName=clavicle_l}}]
    constraintWeights [1,1,1]
   }
   n13 = "Constraint - dyn_wire_bk_MSTR (PPPPP)"!(
    ComponentPose=@n12
    Alpha=1
    ConstraintWeights_0=0.0177
    ConstraintWeights_1:float=0.1556
    ConstraintWeights_2:float=0.5737
    ConstraintWeights_3:float=0.2361
    ConstraintWeights_4:float=0.0169
   ) -> (Pose) {
    alphaBoolBlend {blendOption=Linear}
    boneToModify {boneName=dyn_wire_bk_MSTR}
    constraintSetup [
     {targetBone {boneName=deform_pectoral_fr_l}}
     {targetBone {boneName=spine_04}}
     {targetBone {boneName=spine_05}}
     {targetBone {boneName=clavicle_l}}
     {targetBone {boneName=deform_deltoid_fr_l}}
    ]
    constraintWeights [1,1,1,1,1]
   }
   n14 = "Constraint - dyn_headpiece_01_MSTR (PPPPPPP)"!(
    ComponentPose=@n13
    Alpha=1
    ConstraintWeights_0=0.0162
    ConstraintWeights_1:float=0.0308
    ConstraintWeights_2:float=0.0619
    ConstraintWeights_3:float=0.079
    ConstraintWeights_4:float=0.4228
    ConstraintWeights_5:float=0.3642
    ConstraintWeights_6:float=0.025
   ) -> (Pose) {
    alphaBoolBlend {blendOption=Linear}
    boneToModify {boneName=dyn_headpiece_01_MSTR}
    constraintSetup [
     {targetBone {boneName=clavicle_l}}
     {targetBone {boneName=spine_01}}
     {targetBone {boneName=spine_02}}
     {targetBone {boneName=spine_03}}
     {targetBone {boneName=spine_04}}
     {targetBone {boneName=spine_05}}
     {targetBone {boneName=pelvis}}
    ]
    constraintWeights [1,1,1,1,1,1,1]
   }
   n15 = "Constraint - dyn_clip_fr_01_MSTR (PPPPPPP)"!(
    ComponentPose=@n14
    Alpha=1
    ConstraintWeights_0=0.0123
    ConstraintWeights_1:float=0.0455
    ConstraintWeights_2:float=0.0848
    ConstraintWeights_3:float=0.102
    ConstraintWeights_4:float=0.4264
    ConstraintWeights_5:float=0.2915
    ConstraintWeights_6:float=0.0373
   ) -> (Pose) {
    alphaBoolBlend {blendOption=Linear}
    boneToModify {boneName=dyn_clip_fr_01_MSTR}
    constraintSetup [
     {targetBone {boneName=clavicle_l}}
     {targetBone {boneName=spine_01}}
     {targetBone {boneName=spine_02}}
     {targetBone {boneName=spine_03}}
     {targetBone {boneName=spine_04}}
     {targetBone {boneName=spine_05}}
     {targetBone {boneName=pelvis}}
    ]
    constraintWeights [1,1,1,1,1,1,1]
   }
   n16 = "Transform (Modify) Bone - Bone: dyn_root_simspace"(ComponentPose=@n15, Alpha=1, Translation="0,0,0", Rotation=Get sim_space_rotation() -> (sim_space_rotation:Rotator), Scale="1,1,1") {
    rotationMode=BMM_Replace
    rotationSpace=BCS_WorldSpace
    translationSpace=BCS_BoneSpace
    alphaBoolBlend {blendOption=Linear}
    alphaScaleBiasClamp {bInterpResult=true, interpSpeedDecreasing=2, interpSpeedIncreasing=2}
    boneToModify {boneName=dyn_root_simspace}
   }
   n17 = Get GravityOverride()
   n18 = RigidBody!(ComponentPose=@n16, Alpha, OverrideWorldGravity:Vector=@n17, ExternalForce) -> (Pose) {
    bClampLinearTranslationLimitToRefPose=true
    bOverrideWorldGravity=true
    overridePhysicsAsset=&/M_MED_Dusty_Dyn
    simulationSpace=BaseBoneSpace
    alphaBoolBlend {blendOption=Linear}
    baseBoneRef {boneName=dyn_root_simspace}
    componentAppliedLinearAccClamp=<3000 3000 3000>
    componentLinearAccScale {z=0.5}
    componentLinearVelScale {z=0.5}
    overrideWorldGravity {z=-980}
    simSpaceSettings {
     maxAngularAcceleration=1000
     maxAngularVelocity=3
     maxLinearAcceleration=700
     maxLinearVelocity=500
     velocityScaleZ=0
     worldAlpha=1
     externalLinearDragV=<1 1>
    }
   }
   n19 = RigidBody(ComponentPose=@n18, OverrideWorldGravity="0,0,-980") {
    bClampLinearTranslationLimitToRefPose=true
    bOverrideWorldGravity=true
    overridePhysicsAsset=&/M_MED_Dusty_Dyn_arm_l
    simulationSpace=BaseBoneSpace
    alphaBoolBlend {blendOption=Linear}
    baseBoneRef {boneName=upperarm_l}
    componentAppliedLinearAccClamp=<3000 3000 3000>
    componentLinearAccScale {z=0.5}
    componentLinearVelScale {z=0.5}
    overrideWorldGravity {z=-980}
    simSpaceSettings {
     maxAngularAcceleration=1000
     maxAngularVelocity=3
     maxLinearAcceleration=700
     maxLinearVelocity=500
     velocityScaleZ=0
     worldAlpha=1
     externalLinearDragV=<1 1>
    }
   }
   n20 = RigidBody(ComponentPose=@n19, OverrideWorldGravity="0,0,-980") {
    bClampLinearTranslationLimitToRefPose=true
    bOverrideWorldGravity=true
    overridePhysicsAsset=&/M_MED_Dusty_Dyn_arm_r
    simulationSpace=BaseBoneSpace
    alphaBoolBlend {blendOption=Linear}
    baseBoneRef {boneName=upperarm_r}
    componentAppliedLinearAccClamp=<3000 3000 3000>
    componentLinearAccScale {z=0.5}
    componentLinearVelScale {z=0.5}
    overrideWorldGravity {z=-980}
    simSpaceSettings {
     maxAngularAcceleration=1000
     maxAngularVelocity=3
     maxLinearAcceleration=700
     maxLinearVelocity=500
     velocityScaleZ=0
     worldAlpha=1
     externalLinearDragV=<1 1>
    }
   }
   n21 = RigidBody(ComponentPose=@n20, OverrideWorldGravity=@n17) {
    bClampLinearTranslationLimitToRefPose=true
    bOverrideWorldGravity=true
    overridePhysicsAsset=&/M_MED_Dusty_Dyn_shirt
    simulationSpace=BaseBoneSpace
    alphaBoolBlend {blendOption=Linear}
    baseBoneRef {boneName=dyn_root_simspace}
    componentAppliedLinearAccClamp=<3000 3000 3000>
    componentLinearAccScale {z=0.5}
    componentLinearVelScale {z=0.5}
    overrideWorldGravity {z=-980}
    simSpaceSettings {
     maxAngularAcceleration=1000
     maxAngularVelocity=3
     maxLinearAcceleration=700
     maxLinearVelocity=500
     velocityScaleZ=0
     worldAlpha=1
     externalLinearDragV=<1 1>
    }
   }
   n22 = Get isWearingBackbling()
   n23 = "Transform (Modify) Bone - Bone: dyn_radio_MSTR"!(ComponentPose=@n21, bAlphaBoolEnabled:bool=@n22, Translation="0,0,0", Rotation="0,0,0", Scale="0,0,0") -> (Pose) {
    alphaInputType=Bool
    rotationSpace=BCS_WorldSpace
    scaleMode=BMM_Replace
    translationSpace=BCS_BoneSpace
    alphaBoolBlend {blendOption=Linear}
    alphaScaleBiasClamp {bInterpResult=true, interpSpeedDecreasing=2, interpSpeedIncreasing=2}
    boneToModify {boneName=dyn_radio_MSTR}
    scale=<0 0 0>
   }
   n24 = "Transform (Modify) Bone - Bone: dyn_pouch_bk_r_MSTR"!(ComponentPose=@n23, bAlphaBoolEnabled:bool=@n22, Translation="0,0,0", Rotation="0,0,0", Scale="0,0,0") -> (Pose) {
    alphaInputType=Bool
    rotationSpace=BCS_WorldSpace
    scaleMode=BMM_Replace
    translationSpace=BCS_BoneSpace
    alphaBoolBlend {blendOption=Linear}
    alphaScaleBiasClamp {bInterpResult=true, interpSpeedDecreasing=2, interpSpeedIncreasing=2}
    boneToModify {boneName=dyn_pouch_bk_r_MSTR}
    scale=<0 0 0>
   }
   Output Pose(Result=Component To Local(ComponentPose=@n24)) {layerGroup=None}
  ]}
 EventGraph {nodes [
   n1 = Event BlueprintUpdateAnimation()
   > n2 = Sequence!(execute) -> (then_0, then_1, then_2:Exec)
   n3 = CheckForBackbling()
   > n4 = Set isWearingBackbling(isWearingBackbling=OR Boolean(A=@n3, B=@n3.bIsBackblingLong))
   > n5 = Get Dynamic Material Instance(SkeletalMeshComponent=GetPartSkeletalMeshComponent())
   > SetScalarParameterValue(self=@n5, ParameterName=Hide Element 01, Value="To Float (Boolean)"(InBool=@n4))
   n6 = Gravity Override State Function DEPRECATED(
    execute=@n2.then_1
    CharacterPartAnimInstance=Self-Reference()
    GameTimeInSeconds=GetGameTimeInSeconds()
    GravityOverrideParams=Get Gravity Override Params()
   )
   > Set GravityOverride(GravityOverride=@n6)
   n7 = ComplexSimSpaceFunction DEPRECATED!(
    execute=@n2.then_2
    CharacterPartAnimInstance:Custom Character Part Anim Instance Object Reference=Self-Reference()
    InterpSpeedMax:double=9
    PawnSpeedMax:double=20
    RInterpErrorTolerance:double=1
    DeltaTime:double=@n1.DeltaTimeX
    DebugDraw:bool=false
    sim_space_rotation_in:Rotator=Get sim_space_rotation()
    inComponentSpaceIn:bool=Get InComponentSpace() -> (InComponentSpace:bool)
    inComponentSpaceLastFrameIn:bool=Get InComponentSpaceLastFrame() -> (InComponentSpaceLastFrame:bool)
    lockedOnSimSpaceIn:bool=Get LockedOnSimSpace() -> (LockedOnSimSpace:bool)
    InterpSpeedIn:double=Get InterpSpeed() -> (InterpSpeed:double)
   ) -> (
    then
    sim_space_rotation:Rotator
    inComponentSpaceOut:bool
    inComponentSpaceLastFrameOut:bool
    lockedOnSimSpaceOut:bool
    InterpSpeedOut:double
   )
   > Set sim_space_rotation(sim_space_rotation:Rotator=@n7) -> (Output_Get:Rotator)
   > Set InComponentSpace(InComponentSpace:bool=@n7.inComponentSpaceOut) -> (Output_Get:bool)
   > Set InComponentSpaceLastFrame(InComponentSpaceLastFrame:bool=@n7.inComponentSpaceLastFrameOut) -> (Output_Get:bool)
   > Set LockedOnSimSpace(LockedOnSimSpace:bool=@n7.lockedOnSimSpaceOut) -> (Output_Get:bool)
   > Set InterpSpeed(InterpSpeed:double=@n7.InterpSpeedOut) -> (Output_Get:double)
  ]}
]
variables [
 Gravity Override Params {type=GravityOverrideParamsStruct}
 GravityOverride {type=Vector}
 InComponentSpace {type=boolean}
 InComponentSpaceLastFrame {type=boolean}
 InterpSpeed {type=number}
 isWearingBackbling {type=boolean}
 LockedOnSimSpace {type=boolean}
 sim_space_rotation {type=Rotator}
]
